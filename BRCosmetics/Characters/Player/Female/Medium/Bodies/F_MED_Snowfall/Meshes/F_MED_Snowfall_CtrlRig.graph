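# Corrective Control Rig for the Snowfall female medium body cosmetic that adjusts turtleneck, hood, collar and thigh deform bones procedurally as the body pose changes.
# One forward chain reads Euler angles of head, upper-arm and thigh bones relative to neck, spine or pelvis and remaps them into weights.
# Those weights blend preset translation and rotation offsets, many built from initial bone rotations.
# Some steps are zero-weighted, several nodes are unconnected, and a `test` variable captures a head angle.
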ControlRigBlueprint /BRCosmetics/Characters/Player/Female/Medium/Bodies/F_MED_Snowfall/Meshes/F_MED_Snowfall_CtrlRig
graphs [
 RigVMFunctionLibrary {links [], nodes []}
 RigVMModel {nodes [
   RigUnit_BeginExecution = &RigUnit_BeginExecution()
   n1 = &RigVMFunction_MathQuaternionToEuler(
    RotationOrder=ZYX
    Value=&RigUnit_GetRelativeTransformForItem(Child=(Type=Bone,Name="head"), Parent=(Type=Bone,Name="spine_05"), bChildInitial=false, bParentInitial=true).RelativeTransform.Rotation
   )
   RigVMVariableNode(
    Value=&RigVMFunction_MathFloatAdd(A=@n1.Result.Z) {template="Add::Execute(in A,in B,out Result)"}.Result
    ExecuteContext=@RigUnit_BeginExecution.ExecutePin
   ) {variable=test}
   n2 = &RigVMFunction_MathQuaternionToEuler(
    RotationOrder=ZYX
    Value=/Script/ControlRig.RigUnit_GetInitialBoneTransform(Bone=dyn_turtle_f, Space=LocalSpace) {is=struct RigVMUnitNode}.Transform.Rotation
   )
   > &RigUnit_ModifyTransforms(
    ItemToModify=((Item=(Type=Bone,Name="dyn_turtle_f"),Transform=(Rotation=(X=0,Y=0,Z=0,W=1),Translation=(X=0,Y=-2,Z=-3),Scale3D=(X=1,Y=1,Z=1))))
    Weight=&RigVMFunction_MathFloatRemap(SourceMaximum=70, SourceMinimum=40, bClamp=true, Value=@n1.Result.Z).Result
    ItemToModify.0.Transform.Rotation=&RigVMFunction_MathQuaternionFromEuler(Euler=/Script/RigVM.RigVMFunction_MathVectorAdd(B=(X=-90,Y=70,Z=0), A=@n2.Result) {is=struct RigVMUnitNode, template="Add::Execute(in A,in B,out Result)"}.Result).Result
   )
   n3 = &RigVMFunction_MathQuaternionToEuler(
    RotationOrder=ZYX
    Value=&RigUnit_GetRelativeTransformForItem(Child=(Type=Bone,Name="head"), Parent=(Type=Bone,Name="spine_05"), bChildInitial=false, bParentInitial=false).RelativeTransform.Rotation
   )
   n4 = &RigVMFunction_MathQuaternionToEuler(
    RotationOrder=ZYX
    Value=&RigUnit_GetRelativeTransformForItem(Child=(Type=Bone,Name="head"), Parent=(Type=Bone,Name="spine_05"), bChildInitial=false, bParentInitial=false).RelativeTransform.Rotation
   )
   > /Script/ControlRig.RigUnit_ModifyBoneTransforms(
    BoneToModify=((Bone="dyn_turtle_l",Transform=(Rotation=(X=0,Y=0,Z=0,W=1),Translation=(X=2.5,Y=-2,Z=-3),Scale3D=(X=1,Y=1,Z=1))))
    Mode=AdditiveLocal
    WeightMaximum=1
    WeightMinimum=0
    Weight=&RigVMFunction_MathFloatAdd(
     A=&RigVMFunction_MathFloatRemap(SourceMaximum=-5, SourceMinimum=-50, TargetMaximum=0, TargetMinimum=1, Value=@n3.Result.X).Result
     B=&RigVMFunction_MathFloatRemap(SourceMaximum=-35, SourceMinimum=0, TargetMaximum=-1, TargetMinimum=0, bClamp=true, Value=@n4.Result.Z).Result
    ) {template="Add::Execute(in A,in B,out Result)"}.Result
   ) {is=struct RigVMUnitNode}
   n5 = &RigUnit_GetRelativeTransformForItem(Child=(Type=Bone,Name="head"), Parent=(Type=Bone,Name="spine_05"), bChildInitial=false, bParentInitial=false)
   > &RigUnit_ModifyBoneTransforms(
    BoneToModify=((Bone="dyn_turtle_l",Transform=(Rotation=(X=0,Y=0,Z=0,W=1),Translation=(X=0,Y=0,Z=-2),Scale3D=(X=1,Y=1,Z=1))))
    Mode=AdditiveLocal
    WeightMaximum=1
    WeightMinimum=0
    Weight=&RigVMFunction_MathFloatRemap(
     SourceMaximum=30
     SourceMinimum=0
     TargetMinimum=0
     bClamp=true
     Value=&RigVMFunction_MathQuaternionToEuler(RotationOrder=ZYX, Value=@n5.RelativeTransform.Rotation).Result.Z
    ).Result
   )
   n6 = &RigUnit_GetRelativeTransformForItem(Child=(Type=Bone,Name="head"), Parent=(Type=Bone,Name="spine_05"), bChildInitial=false, bParentInitial=false)
   > &RigUnit_ModifyBoneTransforms(
    BoneToModify=((Bone="dyn_turtle_l",Transform=(Rotation=(X=0,Y=0,Z=0,W=1),Translation=(X=-5,Y=0,Z=0),Scale3D=(X=1,Y=1,Z=1))))
    Mode=AdditiveLocal
    WeightMaximum=1
    WeightMinimum=0
    Weight=&RigVMFunction_MathFloatRemap(
     SourceMaximum=60
     SourceMinimum=10
     TargetMinimum=0
     bClamp=true
     Value=&RigVMFunction_MathQuaternionToEuler(RotationOrder=ZYX, Value=@n6.RelativeTransform.Rotation).Result.X
    ).Result
   )
   n7 = &RigUnit_GetRelativeTransformForItem(Child=(Type=Bone,Name="head"), Parent=(Type=Bone,Name="spine_05"), bChildInitial=false, bParentInitial=false)
   > &RigUnit_ModifyBoneTransforms(
    BoneToModify=((Bone="dyn_turtle_r",Transform=(Rotation=(X=0,Y=0,Z=0,W=1),Translation=(X=-5,Y=0,Z=0),Scale3D=(X=1,Y=1,Z=1))))
    Mode=AdditiveLocal
    WeightMaximum=1
    WeightMinimum=0
    Weight=&RigVMFunction_MathFloatRemap(
     SourceMaximum=-10
     SourceMinimum=-60
     TargetMaximum=0
     TargetMinimum=1
     bClamp=true
     Value=&RigVMFunction_MathQuaternionToEuler(RotationOrder=ZYX, Value=@n7.RelativeTransform.Rotation).Result.X
    ).Result
   )
   n8 = &RigUnit_GetRelativeTransformForItem(Child=(Type=Bone,Name="head"), Parent=(Type=Bone,Name="spine_05"), bChildInitial=false, bParentInitial=false)
   > &RigUnit_ModifyBoneTransforms(
    BoneToModify=((Bone="dyn_turtle_f",Transform=(Rotation=(X=0,Y=0,Z=0,W=1),Translation=(X=0,Y=-5,Z=0),Scale3D=(X=1,Y=1,Z=1))))
    Mode=AdditiveLocal
    WeightMaximum=1
    WeightMinimum=0
    Weight=&RigVMFunction_MathFloatRemap(
     SourceMaximum=-10
     SourceMinimum=-60
     TargetMaximum=0
     TargetMinimum=1
     bClamp=true
     Value=&RigVMFunction_MathQuaternionToEuler(RotationOrder=ZYX, Value=@n8.RelativeTransform.Rotation).Result.X
    ).Result
   )
   n9 = &RigUnit_GetRelativeTransformForItem(Child=(Type=Bone,Name="head"), Parent=(Type=Bone,Name="spine_05"), bChildInitial=false, bParentInitial=false)
   > &RigUnit_ModifyBoneTransforms(
    BoneToModify=((Bone="dyn_turtle_f",Transform=(Rotation=(X=0,Y=0,Z=0,W=1),Translation=(X=0,Y=0,Z=0),Scale3D=(X=1,Y=1,Z=1))))
    Mode=AdditiveLocal
    WeightMaximum=1
    WeightMinimum=0
    Weight=&RigVMFunction_MathFloatRemap(
     SourceMaximum=60
     SourceMinimum=10
     TargetMinimum=0
     bClamp=true
     Value=&RigVMFunction_MathQuaternionToEuler(RotationOrder=ZYX, Value=@n9.RelativeTransform.Rotation).Result.X
    ).Result
   )
   n10 = &RigUnit_GetRelativeTransformForItem(Child=(Type=Bone,Name="head"), Parent=(Type=Bone,Name="neck_02"), bChildInitial=false, bParentInitial=false)
   n11 = &RigVMFunction_MathQuaternionToEuler(RotationOrder=ZYX, Value=&RigUnit_GetInitialBoneTransform(Bone=dyn_hood_base, Space=LocalSpace).Transform.Rotation)
   > &RigUnit_ModifyTransforms(
    ItemToModify=((Item=(Type=Bone,Name="dyn_hood_base"),Transform=(Rotation=(X=0,Y=0,Z=0,W=1),Translation=(X=0,Y=0,Z=0),Scale3D=(X=1,Y=1,Z=1))))
    Weight=&RigVMFunction_MathFloatRemap(
     SourceMaximum=0
     SourceMinimum=-50
     TargetMaximum=0
     TargetMinimum=1
     bClamp=true
     Value=&RigVMFunction_MathQuaternionToEuler(RotationOrder=ZYX, Value=@n10.RelativeTransform.Rotation).Result.Z
    ).Result
    ItemToModify.0.Transform.Rotation=&RigVMFunction_MathQuaternionFromEuler(Euler=&RigVMFunction_MathVectorAdd(B=(X=-90,Y=60,Z=0), A=@n11.Result) {template="Add::Execute(in A,in B,out Result)"}.Result).Result
   )
   n12 = &RigVMFunction_MathQuaternionToEuler(RotationOrder=ZYX, Value=&RigUnit_GetInitialBoneTransform(Bone=dyn_hood_base, Space=LocalSpace).Transform.Rotation)
   > &RigUnit_ModifyTransforms(
    ItemToModify=((Item=(Type=Bone,Name="dyn_hood_base"),Transform=(Rotation=(X=0,Y=0,Z=0,W=1),Translation=(X=0,Y=0,Z=0),Scale3D=(X=1,Y=1,Z=1))))
    Weight=0
    ItemToModify.0.Transform.Rotation=&RigVMFunction_MathQuaternionFromEuler(Euler=&RigVMFunction_MathVectorAdd(B=(X=-60,Y=90,Z=0), A=@n12.Result) {template="Add::Execute(in A,in B,out Result)"}.Result).Result
   )
   n13 = &RigVMFunction_MathQuaternionToEuler(
    RotationOrder=ZYX
    Value=&RigUnit_GetRelativeTransformForItem(Child=(Type=Bone,Name="head"), Parent=(Type=Bone,Name="spine_05"), bChildInitial=false, bParentInitial=false).RelativeTransform.Rotation
   )
   n14 = &RigVMFunction_MathQuaternionToEuler(RotationOrder=ZYX, Value=&RigUnit_GetInitialBoneTransform(Bone=dyn_hood_base, Space=LocalSpace).Transform.Rotation)
   > &RigUnit_ModifyTransforms(
    ItemToModify=((Item=(Type=Bone,Name="dyn_hood_base"),Transform=(Rotation=(X=0,Y=0,Z=0,W=1),Translation=(X=0,Y=0,Z=0),Scale3D=(X=1,Y=1,Z=1))))
    Weight=&RigVMFunction_MathFloatRemap(
     SourceMaximum=-50
     SourceMinimum=-90
     TargetMaximum=0
     TargetMinimum=1
     bClamp=true
     Value=&RigVMFunction_MathFloatAdd(A=@n13.Result.Z) {template="Add::Execute(in A,in B,out Result)"}.Result
    ).Result
    ItemToModify.0.Transform.Rotation=&RigVMFunction_MathQuaternionFromEuler(Euler=&RigVMFunction_MathVectorAdd(B=(X=-90,Y=50,Z=-20), A=@n14.Result) {template="Add::Execute(in A,in B,out Result)"}.Result).Result
   )
   n15 = &RigVMFunction_MathQuaternionToEuler(
    RotationOrder=ZYX
    Value=&RigUnit_GetRelativeTransformForItem(Child=(Type=Bone,Name="spine_05"), Parent=(Type=Bone,Name="upperarm_r"), bChildInitial=false, bParentInitial=false).RelativeTransform.Rotation
   )
   n16 = &RigVMFunction_MathQuaternionToEuler(RotationOrder=ZYX, Value=&RigUnit_GetInitialBoneTransform(Bone=dyn_collar_r, Space=LocalSpace).Transform.Rotation)
   > &RigUnit_ModifyTransforms(
    ItemToModify=((Item=(Type=Bone,Name="dyn_collar_r"),Transform=(Rotation=(X=0,Y=0,Z=0,W=1),Translation=(X=-1,Y=0,Z=0),Scale3D=(X=1,Y=1,Z=1))))
    Weight=&RigVMFunction_MathFloatRemap(SourceMaximum=85, SourceMinimum=30, TargetMinimum=0, bClamp=true, Value=@n15.Result.Z).Result
    ItemToModify.0.Transform.Rotation=&RigVMFunction_MathQuaternionFromEuler(Euler=&RigVMFunction_MathVectorAdd(B=(X=90,Y=-90,Z=0), A=@n16.Result) {template="Add::Execute(in A,in B,out Result)"}.Result).Result
   )
   n17 = &RigUnit_GetRelativeTransformForItem(Child=(Type=Bone,Name="spine_05"), Parent=(Type=Bone,Name="upperarm_l"), bChildInitial=false, bParentInitial=false)
   n18 = &RigVMFunction_MathQuaternionToEuler(RotationOrder=ZYX, Value=&RigUnit_GetInitialBoneTransform(Bone=dyn_collar_l, Space=LocalSpace).Transform.Rotation)
   > &RigUnit_ModifyTransforms(
    ItemToModify=((Item=(Type=Bone,Name="dyn_collar_l"),Transform=(Rotation=(X=0,Y=0,Z=0,W=1),Translation=(X=0,Y=0,Z=0),Scale3D=(X=1,Y=1,Z=1))))
    Weight=&RigVMFunction_MathFloatRemap(
     SourceMaximum=90
     SourceMinimum=50
     TargetMinimum=0
     bClamp=true
     Value=&RigVMFunction_MathQuaternionToEuler(RotationOrder=ZXY, Value=@n17.RelativeTransform.Rotation).Result.Z
    ).Result
    ItemToModify.0.Transform.Rotation=&RigVMFunction_MathQuaternionFromEuler(
     RotationOrder=ZXY
     Euler=&RigVMFunction_MathVectorAdd(B=(X=40,Y=-90,Z=-30), A=@n18.Result) {template="Add::Execute(in A,in B,out Result)"}.Result
    ).Result
   )
   n19 = &RigVMFunction_MathQuaternionToEuler(RotationOrder=ZYX, Value=&RigUnit_GetInitialBoneTransform(Bone=dyn_collar_r, Space=LocalSpace).Transform.Rotation)
   > &RigUnit_ModifyTransforms(
    ItemToModify=((Item=(Type=Bone,Name="dyn_collar_r"),Transform=(Rotation=(X=0,Y=0,Z=0,W=1),Translation=(X=-3,Y=0,Z=0),Scale3D=(X=1,Y=1,Z=1))))
    Weight=0
    ItemToModify.0.Transform.Rotation=&RigVMFunction_MathQuaternionFromEuler(Euler=&RigVMFunction_MathVectorAdd(B=(X=80,Y=-100,Z=-30), A=@n19.Result) {template="Add::Execute(in A,in B,out Result)"}.Result).Result
   )
   n20 = &RigUnit_GetRelativeTransformForItem(Child=(Type=Bone,Name="head"), Parent=(Type=Bone,Name="spine_05"), bChildInitial=false, bParentInitial=false)
   > &RigUnit_ModifyTransforms(
    ItemToModify=((Item=(Type=Bone,Name="dyn_collar_r_02"),Transform=(Rotation=(X=0,Y=0,Z=0,W=1),Translation=(X=-3,Y=-2,Z=-2),Scale3D=(X=1,Y=1,Z=1))))
    WeightMinimum=0
    Weight=&RigVMFunction_MathFloatRemap(
     SourceMaximum=-30
     SourceMinimum=-120
     TargetMaximum=0
     TargetMinimum=1
     bClamp=true
     Value=&RigVMFunction_MathQuaternionToEuler(RotationOrder=ZYX, Value=@n20.RelativeTransform.Rotation).Result.X
    ).Result
   )
   n21 = &RigUnit_GetRelativeTransformForItem(Child=(Type=Bone,Name="head"), Parent=(Type=Bone,Name="spine_05"), bChildInitial=false, bParentInitial=false)
   > &RigUnit_ModifyTransforms(
    ItemToModify=((Item=(Type=Bone,Name="dyn_collal_l_02"),Transform=(Rotation=(X=0,Y=0,Z=0,W=1),Translation=(X=-5,Y=0,Z=0),Scale3D=(X=1,Y=1,Z=1))))
    WeightMinimum=0
    Weight=&RigVMFunction_MathFloatRemap(
     SourceMaximum=0
     SourceMinimum=-30
     TargetMaximum=0
     TargetMinimum=1
     bClamp=true
     Value=&RigVMFunction_MathQuaternionToEuler(RotationOrder=ZYX, Value=@n21.RelativeTransform.Rotation).Result.X
    ).Result
   )
   n22 = &RigUnit_GetRelativeBoneTransform(Bone=thigh_l, Space=pelvis)
   n23 = &RigVMFunction_MathFloatRemap(
    SourceMaximum=100
    SourceMinimum=2
    TargetMinimum=0
    bClamp=true
    Value=/Script/RigVM.RigVMFunction_MathFloatAbs(Value=&RigVMFunction_MathQuaternionToEuler(RotationOrder=ZYX, Value=@n22.Transform.Rotation).Result.Z) {is=struct RigVMUnitNode, template="Absolute::Execute(in Value,out Result)"}.Result
   )
   n24 = &RigVMFunction_MathQuaternionToEuler(RotationOrder=ZYX, Value=&RigUnit_GetInitialBoneTransform(Bone=deform_thigh_fr_l, Space=LocalSpace).Transform.Rotation)
   > &RigUnit_ModifyBoneTransforms(
    BoneToModify=((Bone="deform_thigh_fr_l",Transform=(Rotation=(X=0,Y=0,Z=0,W=1),Translation=(X=10,Y=14,Z=0),Scale3D=(X=1,Y=1,Z=1))))
    Mode=AdditiveLocal
    WeightMaximum=1
    Weight=@n23.Result
    BoneToModify.0.Transform.Rotation=&RigVMFunction_MathQuaternionFromEuler(Euler=&RigVMFunction_MathVectorAdd(B=(X=0,Y=0,Z=20), A=@n24.Result) {template="Add::Execute(in A,in B,out Result)"}.Result).Result
   )
   n25 = &RigVMFunction_MathQuaternionToEuler(RotationOrder=ZYX, Value=&RigUnit_GetInitialBoneTransform(Bone=deform_thigh_bk_l, Space=LocalSpace).Transform.Rotation)
   > &RigUnit_ModifyBoneTransforms(
    BoneToModify=((Bone="deform_thigh_bk_l",Transform=(Rotation=(X=0,Y=0,Z=0,W=1),Translation=(X=33,Y=0,Z=5),Scale3D=(X=1,Y=1,Z=1))))
    Mode=OverrideLocal
    WeightMaximum=1
    WeightMinimum=0
    Weight=@n23.Result
    BoneToModify.0.Transform.Rotation=&RigVMFunction_MathQuaternionFromEuler(Euler=&RigVMFunction_MathVectorAdd(B=(X=0,Y=0,Z=45), A=@n25.Result) {template="Add::Execute(in A,in B,out Result)"}.Result).Result
   )
   n26 = &RigUnit_GetRelativeBoneTransform(Bone=thigh_r, Space=pelvis)
   n27 = &RigVMFunction_MathFloatRemap(
    SourceMaximum=77
    SourceMinimum=177
    TargetMinimum=0
    bClamp=true
    Value=&RigVMFunction_MathFloatAbs(Value=&RigVMFunction_MathQuaternionToEuler(RotationOrder=ZYX, Value=@n26.Transform.Rotation).Result.Z) {template="Absolute::Execute(in Value,out Result)"}.Result
   )
   n28 = &RigVMFunction_MathQuaternionToEuler(RotationOrder=ZYX, Value=&RigUnit_GetInitialBoneTransform(Bone=deform_thigh_fr_r, Space=LocalSpace).Transform.Rotation)
   > &RigUnit_ModifyBoneTransforms(
    BoneToModify=((Bone="deform_thigh_fr_r",Transform=(Rotation=(X=0,Y=0,Z=0,W=1),Translation=(X=-37,Y=-14,Z=0),Scale3D=(X=1,Y=1,Z=1))))
    Mode=OverrideLocal
    WeightMaximum=1
    Weight=@n27.Result
    BoneToModify.0.Transform.Rotation=&RigVMFunction_MathQuaternionFromEuler(Euler=&RigVMFunction_MathVectorAdd(B=(X=0,Y=0,Z=45), A=@n28.Result) {template="Add::Execute(in A,in B,out Result)"}.Result).Result
   )
   n29 = &RigVMFunction_MathQuaternionToEuler(RotationOrder=ZYX, Value=(X=0,Y=0,Z=0,W=1))
   > &RigUnit_ModifyBoneTransforms(
    BoneToModify=((Bone="deform_thigh_bk_r",Transform=(Rotation=(X=0,Y=0,Z=0,W=1),Translation=(X=-60,Y=8,Z=-5),Scale3D=(X=1,Y=1,Z=1))))
    Mode=OverrideLocal
    WeightMaximum=1
    WeightMinimum=0
    Weight=@n27.Result
    BoneToModify.0.Transform.Rotation=&RigVMFunction_MathQuaternionFromEuler(Euler=&RigVMFunction_MathVectorAdd(B=(X=0,Y=0,Z=45), A=@n29.Result) {template="Add::Execute(in A,in B,out Result)"}.Result).Result
   )
   &RigVMFunction_MathFloatRemap(
    SourceMaximum=0
    SourceMinimum=-20
    TargetMaximum=0
    TargetMinimum=1
    bClamp=true
    Value=&RigVMFunction_MathQuaternionToEuler(
     RotationOrder=ZYX
     Value=&RigUnit_GetRelativeTransformForItem(Child=(Type=Bone,Name="head"), Parent=(Type=Bone,Name="neck_02"), bChildInitial=false, bParentInitial=false).RelativeTransform.Rotation
    ).Result.Y
   )
   n30 = &RigVMFunction_MathFloatAdd(A=&RigVMFunction_MathQuaternionToEuler(
     RotationOrder=ZYX
     Value=&RigUnit_GetRelativeTransformForItem(Child=(Type=Bone,Name="spine_05"), Parent=(Type=Bone,Name="clavicle_r"), bChildInitial=false, bParentInitial=false).RelativeTransform.Rotation
    ).Result.Y) {template="Add::Execute(in A,in B,out Result)"}
   &RigVMFunction_MathFloatRemap(SourceMaximum=85, SourceMinimum=78, TargetMinimum=0, bClamp=true, Value=@n30.Result)
   &RigVMFunction_MathFloatRemap(SourceMaximum=-50, SourceMinimum=-80, TargetMinimum=0, bClamp=true, Value=@n30.Result)
   &RigVMFunction_MathFloatRemap(SourceMaximum=-50, SourceMinimum=-80, TargetMinimum=0, bClamp=true, Value=@n15.Result.Y)
   n31 = &RigUnit_GetInitialBoneTransform(Bone=dyn_collar_r_02, Space=LocalSpace)
   &RigVMFunction_MathQuaternionFromEuler(Euler=&RigVMFunction_MathVectorAdd(B=(X=-60,Y=40,Z=-20), A=&RigVMFunction_MathQuaternionToEuler(RotationOrder=ZYX, Value=@n31.Transform.Rotation).Result) {template="Add::Execute(in A,in B,out Result)"}.Result)
   n32 = &RigUnit_GetInitialBoneTransform(Bone=dyn_collal_l_02, Space=LocalSpace)
   &RigVMFunction_MathQuaternionFromEuler(Euler=&RigVMFunction_MathVectorAdd(B=(X=-60,Y=40,Z=-20), A=&RigVMFunction_MathQuaternionToEuler(RotationOrder=ZYX, Value=@n32.Transform.Rotation).Result) {template="Add::Execute(in A,in B,out Result)"}.Result)
   &RigUnit_GetInitialBoneTransform(Bone=deform_thigh_bk_r, Space=LocalSpace)
  ]}
]
hierarchy [
 root {loc=<0 0 0>, rot=<0 0 0 1>}
 pelvis {parent=root, loc=<0 -0.5301 98.0571>, rot=<0.0127 0.707 -0.0127 -0.707>}
 spine_01 {parent=pelvis, loc=<6.1208 -0.9328 0>, rot=<0 0 0.034 -0.9994>}
 spine_02 {parent=spine_01, loc=<6.5996 -0.2095 0>, rot=<0 0 0.0238 -0.9997>}
 spine_03 {parent=spine_02, loc=<6.4914 -0.1237 0>, rot=<0 0 0.0149 -0.9999>}
 spine_04 {parent=spine_03, loc=<6.9598 0.0794 0>, rot=<0 0 -0.0132 -0.9999>}
 spine_05 {parent=spine_04, loc=<6.9987 0 0>, rot=<0 0 -0.0015 -1>}
 clavicle_l {parent=spine_05, loc=<5.4831 -0.1053 -3.2946>, rot=<0.0465 -0.7289 -0.009 -0.6829>}
 upperarm_l {parent=clavicle_l, loc=<9.8658 -1.2136 -0.4105>, rot=<0.0164 -0.3245 -0.0269 -0.9454>}
 lowerarm_l {parent=upperarm_l, loc=<26.465 -0.8345 0.259>, rot=<0.0036 0.0161 -0.2298 -0.9731>}
 hand_l {parent=lowerarm_l, loc=<22.7355 -0.3203 -0.6105>, rot=<-0.6784 -0.0223 0.0735 -0.7307>}
 index_metacarpal_l {parent=hand_l, loc=<1.6535 0.0183 -2.96>, rot=<0.0158 -0.0124 -0.0173 -0.9996>}
 index_01_l {parent=index_metacarpal_l, loc=<8.3008 0.0357 -0.0524>, rot=<-0.0044 0.0485 0.0924 -0.9945>}
 index_02_l {parent=index_01_l, loc=<3.3587 -0.3004 0.0452>, rot=<-0.002 0.0027 0.3225 -0.9466>}
 index_03_l {parent=index_02_l, loc=<2.5578 -0.123 -0.0252>, rot=<-0.002 0.0067 0.2558 -0.9667>}
 middle_metacarpal_l {parent=hand_l, loc=<2.1856 -0.1725 -0.5984>, rot=<0.006 0.0168 -0.0329 -0.9993>}
 middle_01_l {parent=middle_metacarpal_l, loc=<7.3275 -0.2447 0.0129>, rot=<-0.0208 0.0042 0.0858 -0.9961>}
 middle_02_l {parent=middle_01_l, loc=<4.0123 0.1401 0.1175>, rot=<-0.0218 0.0258 0.3512 -0.9357>}
 middle_03_l {parent=middle_02_l, loc=<2.6077 0.0534 0.1373>, rot=<0.0082 0.0188 0.1708 -0.9851>}
 pinky_metacarpal_l {parent=hand_l, loc=<1.7616 -0.9753 3.3814>, rot=<-0.0397 0.0814 -0.069 -0.9935>}
 pinky_01_l {parent=pinky_metacarpal_l, loc=<7.1709 -0.4831 -0.1094>, rot=<-0.0654 -0.0035 0.2242 -0.9723>}
 pinky_02_l {parent=pinky_01_l, loc=<2.8274 0.0893 -0.1043>, rot=<0.0037 0.0011 0.2854 -0.9584>}
 pinky_03_l {parent=pinky_02_l, loc=<2.2046 0.0307 -0.0078>, rot=<-0.0002 0.0098 0.1445 -0.9895>}
 ring_metacarpal_l {parent=hand_l, loc=<2.2074 -0.3625 1.3638>, rot=<-0.0284 0.0722 -0.0458 -0.9959>}
 ring_01_l {parent=ring_metacarpal_l, loc=<7.0295 -0.6197 -0.1446>, rot=<-0.0489 -0.0252 0.1892 -0.9804>}
 ring_02_l {parent=ring_01_l, loc=<3.6174 0.4647 -0.1027>, rot=<0 0 0.2806 -0.9598>}
 ring_03_l {parent=ring_02_l, loc=<2.4152 0.114 0.0199>, rot=<0 0 0.1189 -0.9929>}
 thumb_01_l {parent=hand_l, loc=<1.3118 -0.3191 -2.459>, rot=<0.6546 -0.4042 0.0664 -0.6354>}
 thumb_02_l {parent=thumb_01_l, loc=<5.9336 -0.0448 0.0401>, rot=<-0.0037 0.0213 0.298 -0.9543>}
 thumb_03_l {parent=thumb_02_l, loc=<2.2223 0.3682 -0.0907>, rot=<0.0118 -0.013 0.0066 -0.9998>}
 lowerarm_twist_01_l {parent=lowerarm_l, loc=<18.316 0 0>, rot=<0 0 0 1>}
 lowerarm_twist_02_l {parent=lowerarm_l, loc=<10.6707 0 0>, rot=<0 0 0 1>}
 upperarm_twist_01_l {parent=upperarm_l, loc=<7.2135 -0.1134 0>, rot=<0 0 0 1>}
 upperarm_twist_02_l {parent=upperarm_l, loc=<19.7998 -0.1134 0>, rot=<0 0 0 1>}
 clavicle_r {parent=spine_05, loc=<5.4831 -0.1053 3.2946>, rot=<0.7289 0.0465 -0.6829 0.009>}
 upperarm_r {parent=clavicle_r, loc=<-9.8658 1.2136 0.4104>, rot=<0.0164 -0.3245 -0.0269 -0.9454>}
 lowerarm_r {parent=upperarm_r, loc=<-26.4651 0.8345 -0.259>, rot=<0.0036 0.0161 -0.2298 -0.9731>}
 hand_r {parent=lowerarm_r, loc=<-22.7356 0.3203 0.6106>, rot=<-0.6784 -0.0223 0.0735 -0.7307>}
 index_metacarpal_r {parent=hand_r, loc=<-1.6534 -0.0181 2.9601>, rot=<0.0158 -0.0124 -0.0173 -0.9996>}
 index_01_r {parent=index_metacarpal_r, loc=<-8.3008 -0.0357 0.0524>, rot=<-0.0044 0.0485 0.0924 -0.9945>}
 index_02_r {parent=index_01_r, loc=<-3.3589 0.3003 -0.0453>, rot=<-0.002 0.0027 0.3225 -0.9466>}
 index_03_r {parent=index_02_r, loc=<-2.5578 0.123 0.0252>, rot=<-0.002 0.0067 0.2558 -0.9667>}
 middle_metacarpal_r {parent=hand_r, loc=<-2.1856 0.1726 0.5984>, rot=<0.006 0.0168 -0.0329 -0.9993>}
 middle_01_r {parent=middle_metacarpal_r, loc=<-7.3273 0.2447 -0.0129>, rot=<-0.0208 0.0042 0.0858 -0.9961>}
 middle_02_r {parent=middle_01_r, loc=<-4.0124 -0.14 -0.1175>, rot=<-0.0218 0.0258 0.3512 -0.9357>}
 middle_03_r {parent=middle_02_r, loc=<-2.6077 -0.0533 -0.1373>, rot=<0.0082 0.0188 0.1708 -0.9851>}
 pinky_metacarpal_r {parent=hand_r, loc=<-1.7617 0.9753 -3.3814>, rot=<-0.0397 0.0814 -0.069 -0.9935>}
 pinky_01_r {parent=pinky_metacarpal_r, loc=<-7.1707 0.4832 0.1094>, rot=<-0.0654 -0.0035 0.2242 -0.9723>}
 pinky_02_r {parent=pinky_01_r, loc=<-2.8274 -0.0892 0.1043>, rot=<0.0037 0.0011 0.2854 -0.9584>}
 pinky_03_r {parent=pinky_02_r, loc=<-2.2045 -0.031 0.0078>, rot=<-0.0002 0.0098 0.1445 -0.9895>}
 ring_metacarpal_r {parent=hand_r, loc=<-2.2073 0.3627 -1.3638>, rot=<-0.0284 0.0722 -0.0458 -0.9959>}
 ring_01_r {parent=ring_metacarpal_r, loc=<-7.0297 0.6197 0.1446>, rot=<-0.0489 -0.0252 0.1892 -0.9804>}
 ring_02_r {parent=ring_01_r, loc=<-3.6171 -0.4648 0.1027>, rot=<0 0 0.2806 -0.9598>}
 ring_03_r {parent=ring_02_r, loc=<-2.4153 -0.1141 -0.0199>, rot=<0 0 0.1189 -0.9929>}
 thumb_01_r {parent=hand_r, loc=<-1.3119 0.3192 2.459>, rot=<0.6546 -0.4042 0.0664 -0.6354>}
 thumb_02_r {parent=thumb_01_r, loc=<-5.9335 0.0447 -0.0401>, rot=<-0.0037 0.0213 0.298 -0.9543>}
 thumb_03_r {parent=thumb_02_r, loc=<-2.2223 -0.3682 0.0907>, rot=<0.0118 -0.013 0.0066 -0.9998>}
 lowerarm_twist_01_r {parent=lowerarm_r, loc=<-18.3161 0 0>, rot=<0 0 0 1>}
 lowerarm_twist_02_r {parent=lowerarm_r, loc=<-10.6707 0 0>, rot=<0 0 0 1>}
 upperarm_twist_01_r {parent=upperarm_r, loc=<-7.2135 0.1134 0>, rot=<0 0 0 -1>}
 upperarm_twist_02_r {parent=upperarm_r, loc=<-19.7998 0.1134 0>, rot=<0 0 0 -1>}
 neck_01 {parent=spine_05, loc=<11.3138 0.1529 0>, rot=<0 0 -0.2003 -0.9797>}
 neck_02 {parent=neck_01, loc=<4.5627 0 0>, rot=<0 0 0.0504 -0.9987>}
 head {parent=neck_02, loc=<5.3914 0 0>, rot=<0 0 0.115 -0.9934>}
 dyn_hood_base {parent=spine_05, loc=<8.9218 -8.7765 -0.2207>, rot=<-0.5196 0.4795 0.4795 -0.5196>}
 dyn_hood {parent=dyn_hood_base, loc=<0 0 0>, rot=<0 -0.5544 0 -0.8322>}
 dyn_hood_02 {parent=dyn_hood, loc=<6.6031 0 0>, rot=<-0.001 -0.1573 0.0004 -0.9876>}
 dyn_fur_01 {parent=dyn_hood_02, loc=<3.1396 -9.2847 -1.5134>, rot=<0.7073 -0.028 -0.0283 -0.7058>}
 dyn_fur_02 {parent=dyn_hood_02, loc=<6.4449 0 0>, rot=<0.001 -0.0398 -0.0002 -0.9992>}
 dyn_fur_03 {parent=dyn_hood_02, loc=<3.1327 8.1614 -1.5493>, rot=<0.7073 -0.028 -0.0283 -0.7058>}
 dyn_collar_r {parent=spine_05, loc=<7.0911 6.7009 7.2245>, rot=<0.8651 0.289 0.4044 -0.0672>}
 dyn_collar_l {parent=spine_05, loc=<7.1842 5.7605 -6.9864>, rot=<-0.0696 -0.3657 -0.3462 -0.8611>}
 dyn_turtle_r {parent=spine_05, loc=<16.1156 5.2964 7.0229>, rot=<0.7065 0.0284 0.7065 -0.0284>}
 dyn_turtle_l {parent=spine_05, loc=<16.1156 5.2964 -6.4905>, rot=<0.0284 -0.7065 -0.0284 -0.7065>}
 dyn_turtle_f {parent=spine_05, loc=<11.6739 12.6495 1.0641>, rot=<0.4795 0.5196 0.5196 0.4795>}
 dyn_collar_r_02 {parent=spine_05, loc=<9.6084 1.1462 7.009>, rot=<-0.2287 0.1797 -0.2543 -0.9223>}
 dyn_collal_l_02 {parent=spine_05, loc=<11.0644 -1.4814 -6.8284>, rot=<0.9054 0.2128 -0.2964 0.2169>}
 thigh_l {parent=pelvis, loc=<-1.0855 -0.2109 -8.7053>, rot=<0.0009 0.0355 0.0155 -0.9993>}
 calf_l {parent=thigh_l, loc=<-42.5284 -0.0019 -0.7297>, rot=<0.0006 -0.0161 -0.0391 -0.9991>}
 calf_twist_01_l {parent=calf_l, loc=<-35.8483 -1.9197 -0.5794>, rot=<0.0237 0.0095 -0.0035 -0.9997>}
 calf_twist_02_l {parent=calf_l, loc=<-19.1546 -1.7943 -0.2638>, rot=<0.0237 0.0095 -0.0035 -0.9997>}
 foot_l {parent=calf_l, loc=<-46.7402 -2.5014 -0.592>, rot=<0.0427 -0.0238 0.0353 -0.9982>}
 ball_l {parent=foot_l, loc=<-3.5658 10.8066 -0.1544>, rot=<0.0001 -0.0001 -0.7063 -0.7079>}
 thigh_twist_01_l {parent=thigh_l, loc=<-27.952 0 0>, rot=<0.0013 0 0 -1>}
 dyn_sub_thigh_l {parent=thigh_twist_01_l, loc=<23.1263 0.0216 0.33>, rot=<-0.4815 0.4824 0.5194 -0.5155>}
 deform_thigh_fr_l {parent=thigh_twist_01_l, loc=<27.767 0 0>, rot=<-0.0015 -0.0355 0.0024 -0.9994>}
 deform_thigh_bk_l {parent=thigh_twist_01_l, loc=<23.1286 0 0>, rot=<0.0014 -0.0071 0.0005 -1>}
 thigh_r {parent=pelvis, loc=<-1.0853 -0.2109 8.7053>, rot=<0.0355 -0.0009 0.9993 0.0155>}
 calf_r {parent=thigh_r, loc=<42.5286 0.0019 0.7297>, rot=<0.0006 -0.0161 -0.0391 -0.9991>}
 calf_twist_01_r {parent=calf_r, loc=<35.8485 1.9197 0.5794>, rot=<0.0237 0.0095 -0.0035 -0.9997>}
 calf_twist_02_r {parent=calf_r, loc=<19.1546 1.7943 0.2638>, rot=<0.0237 0.0095 -0.0035 -0.9997>}
 foot_r {parent=calf_r, loc=<46.7403 2.5014 0.592>, rot=<0.0427 -0.0238 0.0353 -0.9982>}
 ball_r {parent=foot_r, loc=<3.5658 -10.8066 0.1544>, rot=<0.0001 -0.0001 -0.7063 -0.7079>}
 thigh_twist_01_r {parent=thigh_r, loc=<27.952 0 0>, rot=<0.0013 0 0 -1>}
 dyn_sub_thigh_r {parent=thigh_twist_01_r, loc=<-23.1268 -0.0216 -0.33>, rot=<0.5194 0.5155 0.4815 0.4824>}
 deform_thigh_fr_r {parent=thigh_twist_01_r, loc=<-27.767 0 0>, rot=<-0.0015 -0.0355 0.0024 -0.9994>}
 deform_thigh_bk_r {parent=thigh_twist_01_r, loc=<-23.1268 0 0>, rot=<0.5194 0.5155 0.4815 0.4824>}
]
variables [test {type=double, default=0}]
