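# Physics asset for the hood of the F_MED_Snowfall female medium player body, defining capsule bodies for the hood, its fur trim and the head.
# Tightly limited constraints chain dyn_hood_base to dyn_hood to dyn_hood_02, from which three fur bodies branch with slightly wider, symmetric swing.
# Twist is nearly locked on every joint, and the hood segments swing more on one axis than the other.
# The head capsule has no constraint.
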PhysicsAsset /BRCosmetics/Characters/Player/Female/Medium/Bodies/F_MED_Snowfall/Meshes/F_MED_Snowfall_Hood_Physics
bodies {
 dyn_fur_01 [{shapeType=Capsule, length=3, radius=3, shapeName=dyn_fur_02_capsule, center=<3 0 0>, rotation=<rot 90 0 0>}]
 dyn_fur_02 [{shapeType=Capsule, length=3, radius=3, shapeName=dyn_fur_02_capsule, center=<3 0 0>, rotation=<rot 90 0 0>}]
 dyn_fur_03 [{shapeType=Capsule, length=3, radius=3, shapeName=dyn_fur_02_capsule, center=<3 0 0>, rotation=<rot 90 0 0>}]
 dyn_hood [{shapeType=Capsule, length=9.5, radius=6, shapeName=dyn_hood_capsule, center=<0.5 0 0>, rotation=<rot 0 0 90>}]
 dyn_hood_02 [{shapeType=Capsule, length=6.875, radius=5, shapeName=dyn_hood_02_capsule, center=<4 0 0>, rotation=<rot 90 0 0>}]
 dyn_hood_base [{shapeType=Capsule, length=0.505, radius=0.505, shapeName=dyn_hood_base_capsule, rotation=<rot 0 -90 0>}]
 head [{shapeType=Capsule, length=7.25, radius=7, shapeName=head_capsule, center=<2.9103 0.0414 0>, rotation=<rot 0 55 90>}]
}
constraints [
 {bone1Name=dyn_hood, bone2Name=dyn_hood_base, swing1LimitDegrees=1, swing2LimitDegrees=10, twistLimitDegrees=1}
 {bone1Name=dyn_fur_02, bone2Name=dyn_hood_02, swing1LimitDegrees=12, swing2LimitDegrees=12, twistLimitDegrees=1}
 {bone1Name=dyn_hood_02, bone2Name=dyn_hood, swing1LimitDegrees=3, swing2LimitDegrees=16, twistLimitDegrees=1}
 {bone1Name=dyn_fur_03, bone2Name=dyn_hood_02, swing1LimitDegrees=12, swing2LimitDegrees=12, twistLimitDegrees=1}
 {bone1Name=dyn_fur_01, bone2Name=dyn_hood_02, swing1LimitDegrees=12, swing2LimitDegrees=12, twistLimitDegrees=1}
]
