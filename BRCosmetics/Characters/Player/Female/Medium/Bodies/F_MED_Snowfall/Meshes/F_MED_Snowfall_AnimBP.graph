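# Post-process style Anim Blueprint for the F_MED_Snowfall character body, correcting hood, collar, thigh and backpack bones over the parent mesh's copied pose.
# A Control Rig node feeds chained bone offsets and a hood RigidBody simulation.
# The event graph drives their alphas from head and thigh angles, crouch, skydive, melee weapon and backpack state.
# Winged backpacks are detected by matching the attached component's name against a hardcoded list.
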AnimBlueprint /BRCosmetics/Characters/Player/Female/Medium/Bodies/F_MED_Snowfall/Meshes/F_MED_Snowfall_AnimBP
target_skeleton=&/F_MED_Snowfall_Skeleton
graphs [
 AnimGraph {nodes [
   n1 = Control Rig!(Source=Copy Pose From Mesh() {bUseAttachedParent=true}, Alpha, test:double=0) -> (Pose) {
    alphaBoolBlend {blendOption=Linear}
    controlRigAssetReference {blueprintRigClass=&/F_MED_Snowfall_CtrlRig.F_MED_Snowfall_CtrlRig_C}
   }
   n2 = "Transform (Modify) Bone - Bone: dyn_turtle_r"(
    ComponentPose=Local To Component(LocalPose=@n1)
    Alpha=Get Neck Alpha() -> (Neck Alpha:double)
    Translation="-1,-0.000,-1"
    Rotation="30,0.000,0.000"
    Scale="1,1,1"
   ) {
    rotationMode=BMM_Additive
    translationMode=BMM_Additive
    alphaBoolBlend {blendOption=Linear}
    alphaScaleBiasClamp {bInterpResult=true, interpSpeedDecreasing=30, interpSpeedIncreasing=30}
    boneToModify {boneName=dyn_turtle_r}
    rotation {pitch=30}
    translation {x=-1, z=-1}
   }
   n3 = "Transform (Modify) Bone - Bone: dyn_hood_base"(ComponentPose=@n2, Alpha=Get IsCrouchingAnim() -> (IsCrouchingAnim:double), Translation="0,0,0", Rotation="0,0,-27", Scale="1,1,1") {
    rotationMode=BMM_Additive
    translationMode=BMM_Additive
    alphaBoolBlend {blendOption=Linear}
    alphaScaleBiasClamp {bInterpResult=true, interpSpeedDecreasing=30, interpSpeedIncreasing=30}
    boneToModify {boneName=dyn_hood_base}
    rotation {roll=-27}
   }
   n4 = "Transform (Modify) Bone - Bone: dyn_hood_base"(ComponentPose=@n3, Alpha=Get IsSkyDivingAnim() -> (IsSkyDivingAnim:double), Translation="0,0,0", Rotation="0,0,-30", Scale="1,1,1") {
    rotationMode=BMM_Additive
    translationMode=BMM_Additive
    alphaBoolBlend {blendOption=Linear}
    alphaScaleBiasClamp {bInterpResult=true, interpSpeedDecreasing=30, interpSpeedIncreasing=30}
    boneToModify {boneName=dyn_hood_base}
    rotation {roll=-30}
   }
   n5 = "Transform (Modify) Bone - Bone: dyn_collar_r_02"(
    ComponentPose=@n4
    Alpha=Get IsOneHandMeleeAnim() -> (IsOneHandMeleeAnim:double)
    Translation="-1.3,-1,-0.5"
    Rotation="0,0,0"
    Scale="0.8,1,1"
   ) {
    rotationMode=BMM_Additive
    translationMode=BMM_Additive
    translationSpace=BCS_BoneSpace
    alphaBoolBlend {blendOption=Linear}
    alphaScaleBiasClamp {bInterpResult=true, interpSpeedDecreasing=30, interpSpeedIncreasing=30}
    boneToModify {boneName=dyn_collar_r_02}
    scale {x=0.8}
    translation=<-1.3 -1 -0.5>
   }
   n6 = "Transform (Modify) Bone - Bone: dyn_sub_thigh_r"(ComponentPose=@n5, Alpha=Get legSwingDelta_r() -> (legSwingDelta_r:double), Translation="0,0,0", Rotation="0,0,0", Scale="6,0,0") {
    scaleMode=BMM_Additive
    scaleSpace=BCS_BoneSpace
    alphaBoolBlend {blendOption=Linear}
    boneToModify {boneName=dyn_sub_thigh_r}
    scale=<6 0 0>
   }
   n7 = "Transform (Modify) Bone - Bone: dyn_sub_thigh_l"(ComponentPose=@n6, Alpha=Get legSwingDelta_l() -> (legSwingDelta_l:double), Translation="0,0,0", Rotation="0,0,0", Scale="6,0,0") {
    scaleMode=BMM_Additive
    scaleSpace=BCS_BoneSpace
    alphaBoolBlend {blendOption=Linear}
    boneToModify {boneName=dyn_sub_thigh_l}
    scale=<6 0 0>
   }
   n8 = RigidBody(ComponentPose=@n7, Alpha=Get RigidBodyAlpha() -> (RigidBodyAlpha:double)) {
    bClampLinearTranslationLimitToRefPose=true
    bOverrideWorldGravity=true
    overridePhysicsAsset=&/F_MED_Snowfall_Hood_Physics
    simulationSpace=BaseBoneSpace
    alphaBoolBlend {blendOption=Linear}
    baseBoneRef {boneName=spine_05}
    componentAppliedLinearAccClamp=<6000 6000 6000>
    componentLinearAccScale=<0.2 0.2 0.45>
    componentLinearVelScale=<0.2 0.2 0.45>
    overrideWorldGravity {z=-980}
   }
   n9 = Get IswearingbackpackAnim() -> (IswearingbackpackAnim:double)
   n10 = "Transform (Modify) Bone - Bone: dyn_mstr_Backpack"(ComponentPose=@n8, Alpha=@n9, Translation="0,-5,0", Rotation="0,0.000,-15", Scale="1,1,1") {
    rotationMode=BMM_Additive
    rotationSpace=BCS_BoneSpace
    translationMode=BMM_Additive
    translationSpace=BCS_BoneSpace
    alphaBoolBlend {blendOption=Linear}
    boneToModify {boneName=dyn_mstr_Backpack}
    rotation {roll=-15}
    translation {y=-5}
   }
   n11 = "Transform (Modify) Bone - Bone: dyn_hood_base"(ComponentPose=@n10, Alpha=@n9, Translation="-1.5,0,0", Rotation="0,0,0", Scale="0.8,0.8,0.8") {
    rotationMode=BMM_Additive
    rotationSpace=BCS_BoneSpace
    scaleMode=BMM_Additive
    scaleSpace=BCS_BoneSpace
    translationMode=BMM_Additive
    translationSpace=BCS_BoneSpace
    alphaBoolBlend {blendOption=Linear}
    boneToModify {boneName=dyn_hood_base}
    scale=<0.8 0.8 0.8>
    translation {x=-1.5}
   }
   n12 = "Transform (Modify) Bone - Bone: dyn_mstr_Backpack"(ComponentPose=@n11, Alpha=Get IsBackPackSkyDive() -> (IsBackPackSkyDive:double), Translation="0,-5,-1", Rotation="0,0,0", Scale="1,1,1") {
    rotationMode=BMM_Additive
    rotationSpace=BCS_BoneSpace
    translationMode=BMM_Additive
    translationSpace=BCS_BoneSpace
    alphaBoolBlend {blendOption=Linear}
    alphaScaleBiasClamp {bInterpResult=true}
    boneToModify {boneName=dyn_mstr_Backpack}
    translation {y=-5, z=-1}
   }
   n13 = "Transform (Modify) Bone - Bone: dyn_mstr_Backpack"(
    ComponentPose=@n12
    Alpha=Get BackPackTiltCrouch() -> (BackPackTiltCrouch:double)
    Translation="-0.000,-3,4"
    Rotation="0,0,-25"
    Scale="1,1,1"
   ) {
    rotationMode=BMM_Additive
    rotationSpace=BCS_BoneSpace
    translationMode=BMM_Additive
    translationSpace=BCS_BoneSpace
    alphaBoolBlend {blendOption=Linear}
    alphaScaleBiasClamp {bInterpResult=true}
    boneToModify {boneName=dyn_mstr_Backpack}
    rotation {roll=-25}
    translation {y=-3, z=4}
   }
   n14 = Get IsWearingWingsAnim() -> (IsWearingWingsAnim:double)
   n15 = "Transform (Modify) Bone - Bone: dyn_mstr_Backpack"(ComponentPose=@n13, Alpha=@n14, Translation="-0.000,1,0", Rotation="0,-0.000,-10.000", Scale="1,1,1") {
    rotationMode=BMM_Additive
    rotationSpace=BCS_BoneSpace
    translationMode=BMM_Additive
    translationSpace=BCS_BoneSpace
    alphaBoolBlend {blendOption=Linear}
    alphaScaleBiasClamp {bInterpResult=true}
    boneToModify {boneName=dyn_mstr_Backpack}
    rotation {roll=-10}
    translation {y=1}
   }
   n16 = "Transform (Modify) Bone - Bone: dyn_hood_base"(ComponentPose=@n15, Alpha=@n14, Translation="-1,0,0.5", Rotation="0,0,0", Scale="0.8,0.8,0.8") {
    rotationMode=BMM_Additive
    rotationSpace=BCS_BoneSpace
    scaleMode=BMM_Additive
    scaleSpace=BCS_BoneSpace
    translationMode=BMM_Additive
    translationSpace=BCS_BoneSpace
    alphaBoolBlend {blendOption=Linear}
    boneToModify {boneName=dyn_hood_base}
    scale=<0.8 0.8 0.8>
    translation {x=-1, z=0.5}
   }
   Output Pose(Result=Component To Local(ComponentPose=@n16)) {layerGroup=None}
  ]}
 EventGraph {nodes [
   Event BlueprintUpdateAnimation()
   > n1 = Sequence!(execute) -> (then_0, then_1, then_2:Exec, then_3:Exec)
   n2 = GetOwnerSkeletalMeshComponent()
   > Set legSwingDelta_r(legSwingDelta_r:double=MapRangeClamped(Value=GetTwistAndSwingAngleOfDeltaRotationFromRefPose(self=@n2, BoneName=thigh_r).OutSwingAngle, InRangeA=150, InRangeB=75)) -> (Output_Get:double)
   > Set legSwingDelta_l(legSwingDelta_l:double=MapRangeClamped(Value=GetTwistAndSwingAngleOfDeltaRotationFromRefPose(self=@n2, BoneName=thigh_l).OutSwingAngle, InRangeA=43, InRangeB=110)) -> (Output_Get:double)
   > Set IsCrouchingAnim(IsCrouchingAnim:double="To Float (Boolean)"(InBool=Get bIsCrouching())) -> (Output_Get:double)
   > Set IsCrouchRunningAnim(IsCrouchRunningAnim:double="To Float (Boolean)"(InBool=Get bIsCrouchSprinting())) -> (Output_Get:double)
   > Set IsSkyDivingAnim(IsSkyDivingAnim:double="To Float (Boolean)"(InBool=Get bIsSkyDiving())) -> (Output_Get:double)
   > Branch(Condition=Get bIsInFrontEnd())
   n3 = BreakTransform(InTransform=GetDeltaTransformFromRefPose(self=GetOwnerSkeletalMeshComponent(), BoneName=head, BaseName=spine_05))
   > Set Neck Alpha(Neck Alpha:double=MapRangeClamped(Value=BreakRotator(InRot=@n3.Rotation), InRangeA=-20, InRangeB=0, OutRangeA=1, OutRangeB=0)) -> (Output_Get:double)
   n4 = Cast To FortPlayerAnimInstance(execute=@n1.then_1, Object=GetAnimInstance(self=GetOwnerSkeletalMeshComponent()))
   > Set IsOneHandMeleeAnim(IsOneHandMeleeAnim:double=0) -> (Output_Get:double)
   > n5 = Switch on EFortWeaponCoreAnimation!(execute, Selection:EFortWeaponCoreAnimation Enum=Get WeaponCoreAnim(self=@n4)) -> (
    Melee:Exec
    Pistol:Exec
    Shotgun:Exec
    PaperBlueprint:Exec
    Rifle:Exec
    MeleeOneHand:Exec
    MachinePistol:Exec
    RocketLauncher:Exec
    GrenadeLauncher:Exec
    GoingCommando:Exec
    AssaultRifle:Exec
    TacticalShotgun:Exec
    SniperRifle:Exec
    TrapPlacement:Exec
    ShoulderLauncher:Exec
    AbilityDecoTool:Exec
    Crossbow:Exec
    C4:Exec
    RemoteControl:Exec
    DualWield:Exec
    AR_BullPup:Exec
    AR_ForwardGrip:Exec
    MedPackPaddles:Exec
    SMG_P90:Exec
    AR_DrumGun:Exec
    Consumable_Small:Exec
    Consumable_Large:Exec
    Balloon:Exec
    MountedTurret:Exec
    CreativeTool:Exec
    ExplosiveBow:Exec
    AshtonIndigo:Exec
    AshtonChicago:Exec
    MeleeDualWield:Exec
    Unarmed:Exec
   )
   Branch(execute=[@n5.AssaultRifle, @n5.Crossbow, @n5.GrenadeLauncher, @n5.Rifle, @n5.RocketLauncher, @n5.Shotgun, @n5.SniperRifle, @n5.TacticalShotgun])
   Branch(execute=[@n5.MachinePistol, @n5.Pistol])
   Branch(execute=@n5.MeleeOneHand)
   > Set IsOneHandMeleeAnim(IsOneHandMeleeAnim=1)
   Set IsWearingWings(execute=@n1.then_2, IsWearingWings:bool=false) -> (Output_Get:bool)
   n6 = FindChildCommponentFromAttachSocket!(
    self:Custom Character Part Anim Instance Object Reference
    Parent:Skeletal Mesh Component Object Reference=GetPartSkeletalMeshComponent()
    bIncludeAllDescendants:bool=true
    AttachSocket:Name=Backpack_BR
   ) -> (ReturnValue:Skeletal Mesh Component Object Reference)
   n7 = Split(SourceString=GetDisplayName(Object=@n6), InStr=Append(A=GetObjectName(Object=@n6), B=" "), SearchDir=FromStart)
   > Branch(Condition=OR Boolean!(
     A="Equal, Case Insensitive (String)"(A=@n7.RightS, B=F_MED_Snowfall_Pack)
     B="Equal, Case Insensitive (String)"(A=@n7.RightS, B=Backpack_F_RavenQuill)
     C:bool="Equal, Case Insensitive (String)"(A=@n7.RightS, B=F_MED_Angel_Backpack)
     D:bool="Equal, Case Insensitive (String)"(A=@n7.RightS, B=F_MED_Angel_Dark_Pack)
     E:bool="Equal, Case Insensitive (String)"(A=@n7.RightS, B=F_MED_Aztec_Backpack)
     F:bool="Equal, Case Insensitive (String)"(A=@n7.RightS, B=F_MED_Commando_DarkViking_Backpack)
     G:bool="Equal, Case Insensitive (String)"(A=@n7.RightS, B=F_MED_Commando_DarkViking_Backpack_FX)
     H:bool="Equal, Case Insensitive (String)"(A=@n7.RightS, B=F_MED_Cupid_Backpack)
     I:bool="Equal, Case Insensitive (String)"(A=@n7.RightS, B=F_MED_Demon_Pack)
     J:bool="Equal, Case Insensitive (String)"(A=@n7.RightS, B=F_MED_Devil_Rock_BackPack)
     K:bool="Equal, Case Insensitive (String)"(A=@n7.RightS, B=F_MED_Frankie_Pack)
     L:bool="Equal, Case Insensitive (String)"(A=@n7.RightS, B=F_MED_Graffiti_Future)
     M:bool="Equal, Case Insensitive (String)"(A=@n7.RightS, B=F_MED_Jail_Bird_Bumble_Pack)
     N:bool="Equal, Case Insensitive (String)"(A=@n7.RightS, B=F_MED_Punk_Devil_Pack)
     O:bool="Equal, Case Insensitive (String)"(A=@n7.RightS, B=F_MED_Raspberry_Pack)
     P:bool="Equal, Case Insensitive (String)"(A=@n7.RightS, B=F_MED_Shatter_Fly_Pack)
     Q:bool="Equal, Case Insensitive (String)"(A=@n7.RightS, B=F_MED_Shiny_Pack)
     R:bool="Equal, Case Insensitive (String)"(A=@n7.RightS, B=F_MED_Snow_Fairy)
     S:bool="Equal, Case Insensitive (String)"(A=@n7.RightS, B=F_MED_Spooky_Neon_Pack)
     T:bool="Equal, Case Insensitive (String)"(A=@n7.RightS, B=F_MED_Vampire_Backpack)
     U:bool="Equal, Case Insensitive (String)"(A=@n7.RightS, B=F_MED_Winged_Fury_Pack)
     V:bool="Equal, Case Insensitive (String)"(A=@n7.RightS, B=F_MED_Valentine_Dark_Pack)
     W:bool="Equal, Case Insensitive (String)"(A=@n7.RightS, B=Female_Ninja_08)
     X:bool
     Y:bool
    ) -> (ReturnValue))
   > n8 = Set IsWearingWings(IsWearingWings=true)
   n9 = Branch(execute=@n1.then_3, Condition=Get IsWearingWings() -> (IsWearingWings:bool))
   > Set IsWearingWingsAnim(IsWearingWingsAnim:double=1) -> (Output_Get:double)
   Set IsWearingWingsAnim(execute=@n9.else, IsWearingWingsAnim=0)
   n10 = Get isWearingBackpack() -> (isWearingBackpack:bool)
   > Set IswearingbackpackAnim(IswearingbackpackAnim:double="To Float (Boolean)"(InBool=@n10)) -> (Output_Get:double)
   > Set RigidBodyAlpha(RigidBodyAlpha:double=SelectFloat(A=0.5, B=1, bPickA=@n10)) -> (Output_Get:double)
   n11 = OR Boolean!(A=Get bIsSkydiveDiveMode(), B=Get bIsSkydiveDiveMode(), C=Get bIsSkydiveFloating()) -> (ReturnValue)
   > Set IsBackPackSkyDive(IsBackPackSkyDive:double=SelectFloat(A=1, B=0, bPickA=AND Boolean(A=@n11, B=@n10))) -> (Output_Get:double)
   n12 = GetOwnerSkeletalMeshComponent()
   n13 = BreakRotator(InRot=BreakTransform(InTransform=GetDeltaTransformFromRefPose(self=@n12, BoneName=head, BaseName=spine_05)).Rotation)
   n14 = OR Boolean!(A=Get bIsCrouching(), B=Get bIsCrouchSprinting()) -> (ReturnValue)
   > Set BackPackTiltCrouch(BackPackTiltCrouch:double=SelectFloat(A=MapRangeClamped(Value=@n13.Pitch, InRangeA=0, InRangeB=20), B=0, bPickA=AND Boolean(A=@n10, B=@n14))) -> (Output_Get:double)
   Event OnCharacterCustomizationLoaded()
   n15 = CheckForBackbling()
   > Set isWearingBackpack(isWearingBackpack:bool=OR Boolean!(A=@n15, B=@n15.bIsBackblingACape, C:bool=@n15.bIsBackblingLong) -> (ReturnValue)) -> (Output_Get:bool)
   n16 = GetDeltaTransformFromRefPose(self=GetOwnerSkeletalMeshComponent(), BoneName=head, BaseName=spine_05)
   MapRangeClamped(Value=BreakRotator(InRot=BreakTransform(InTransform=@n16).Rotation).Yaw, InRangeA=-20, InRangeB=0)
   PrintString!(execute, InString="To String (Boolean)"(InBool=@n8), bPrintToScreen, bPrintToLog, TextColor, Duration, Key) -> (then)
   PrintString(InString="To String (Boolean)"(InBool=@n8))
  ]}
]
variables [
 BackPackTiltCrouch {type=number}
 IsBackPackSkyDive {type=number}
 IsCrouchingAnim {type=number}
 IsCrouchRunningAnim {type=number}
 IsOneHandMeleeAnim {type=number}
 IsSkyDivingAnim {type=number}
 isWearingBackpack {type=boolean}
 IswearingbackpackAnim {type=number}
 IsWearingWings {type=boolean}
 IsWearingWingsAnim {type=number}
 legSwingDelta_l {type=number}
 legSwingDelta_r {type=number}
 Neck Alpha {type=number}
 NewVar_0 {type=number}
 RigidBodyAlpha {type=number}
]
