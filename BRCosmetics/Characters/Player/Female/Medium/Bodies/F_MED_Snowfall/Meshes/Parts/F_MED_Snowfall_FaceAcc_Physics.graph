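# Physics asset for the F_MED_Snowfall face accessory mesh, setting up secondary motion for the beanie, hair and strands.
# Angle-limited chains run from small base spheres through capsules, with looser right-side hair that branches at its parent.
# Head, neck, upper spine, clavicle and upper-arm capsules supply body collision shapes.
# Collar and turtleneck bodies attach to spine_05 with free rotation, and each strand tip is also constrained to its side's hair chain.
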PhysicsAsset /BRCosmetics/Characters/Player/Female/Medium/Bodies/F_MED_Snowfall/Meshes/Parts/F_MED_Snowfall_FaceAcc_Physics
bodies {
 clavicle_l [{
   shapeType=Capsule
   length=9.755
   radius=3.755
   shapeName=clavicle_l_capsule
   center=<4.1606 0 0.7247>
   rotation=<rot 78.9737 -179.9997 -179.9997>
  }]
 clavicle_r [{
   shapeType=Capsule
   length=10.505
   radius=5.005
   shapeName=clavicle_l_capsule
   center=<-3.078 0 -1.3263>
   rotation=<rot 78.9737 -179.9997 -179.9997>
  }]
 dyn_beanie [{shapeType=Capsule, length=1.4037, radius=2.3187, shapeName=dyn_beanie_capsule, center=<0.2712 0 0>, rotation=<rot 0 -90 90>}]
 dyn_beanie_base [{shapeType=Sphere, radius=0.5, shapeName=dyn_beanie_base_sphere}]
 dyn_collar_l [{
   shapeType=Capsule
   length=9.895
   radius=2.6988
   shapeName=dyn_collar_l_capsule
   center=<2.7208 -1.1731 0.1309>
   rotation=<rot 0.4183 172.7209 -88.8505>
  }]
 dyn_collar_r [{
   shapeType=Capsule
   length=12.895
   radius=1.6988
   shapeName=dyn_collar_l_capsule
   center=<1.9439 3.2404 0.1651>
   rotation=<rot -49.1419 -160.0636 -99.8445>
  }]
 dyn_hair_01_l [{shapeType=Capsule, length=5.7886, radius=2.3124, shapeName=dyn_hair_01_l_capsule, center=<1.7399 0 -0.4012>, rotation=<rot 90 0 0>}]
 dyn_hair_01_r [
  {shapeType=Capsule, length=1.1531, radius=2.0003, shapeName=dyn_hair_01_r_capsule, center=<2.0521 0 0.5427>, rotation=<rot 90 0 0>}
  {shapeType=Capsule, length=2, radius=2, shapeName=dyn_hair_01_r_capsule, center=<2.6293 4.6064 1.6463>, rotation=<rot 70 180 179.9999>}
 ]
 dyn_hair_02_l [{shapeType=Capsule, length=1.1958, radius=2.2711, shapeName=dyn_hair_02_l_capsule, center=<2.1877 0 0>, rotation=<rot 90 0 0>}]
 dyn_hair_02_r [{shapeType=Capsule, length=0.2764, radius=1.9605, shapeName=dyn_hair_02_r_capsule, center=<1.5511 0 0>, rotation=<rot 90 0 0>}]
 dyn_hair_03_l [{
   shapeType=Capsule
   length=0.9818
   radius=2.2681
   shapeName=dyn_hair_03_l_capsule
   center=<1.2457 -0.5001 0.7098>
   rotation=<rot 65.5283 180 180>
  }]
 dyn_hair_03_r [{shapeType=Capsule, length=0.5186, radius=1.8181, shapeName=dyn_hair_03_r_capsule, center=<1.2722 0 0>, rotation=<rot 90 0 0>}]
 dyn_hair_04_l [{shapeType=Capsule, length=1.3377, radius=2.0074, shapeName=dyn_hair_04_l_capsule, center=<1.709 0 0>, rotation=<rot 90 0 0>}]
 dyn_hair_04_r [{shapeType=Capsule, length=1.564, radius=1.4049, shapeName=dyn_hair_04_r_capsule, center=<1.5043 0 0>, rotation=<rot 90 0 0>}]
 dyn_hair_05_r [{shapeType=Capsule, length=0.9481, radius=1.2083, shapeName=dyn_hair_05_r_capsule, center=<1.2093 0 0>, rotation=<rot 90 0 0>}]
 dyn_hair_12_r [{shapeType=Capsule, length=7.75, radius=2.75, shapeName=dyn_hair_12_r_capsule, center=<-0.5583 1.1437 0>, rotation=<rot -80 9.9996 0>}]
 dyn_hair_base_l [{shapeType=Sphere, radius=0.5, shapeName=dyn_hair_base_l_sphere}]
 dyn_hair_base_r [{shapeType=Sphere, radius=0.5, shapeName=dyn_hair_base_r_sphere}]
 dyn_hair_parent_l [{shapeType=Capsule, length=2.1231, radius=3.4284, shapeName=dyn_hair_parent_l_capsule, center=<2.7099 0 0>, rotation=<rot 90 0 0>}]
 dyn_hair_parent_r [{shapeType=Capsule, length=1.425, radius=1.8928, shapeName=dyn_hair_parent_r_capsule, center=<1.9188 0 0>, rotation=<rot 90 0 0>}]
 dyn_strand_01_l [{shapeType=Capsule, length=1.9572, radius=1.2984, shapeName=dyn_strand_01_l_capsule, center=<1.6021 0 0>, rotation=<rot 90 0 0>}]
 dyn_strand_01_r [{
   shapeType=Capsule
   length=1.3567
   radius=0.8459
   shapeName=dyn_strand_01_r_capsule
   center=<1.2057 -0.296 0>
   rotation=<rot 90 180 -154.5443>
  }]
 dyn_strand_02_l [{shapeType=Capsule, length=1.3909, radius=1.2267, shapeName=dyn_strand_02_l_capsule, center=<1.5879 0 0>, rotation=<rot 90 0 0>}]
 dyn_strand_02_r [{shapeType=Capsule, length=1.3813, radius=0.6063, shapeName=dyn_strand_02_r_capsule, center=<0.8069 0 0>, rotation=<rot 90 0 0>}]
 dyn_strand_base_l [{shapeType=Sphere, radius=0.5, shapeName=dyn_strand_base_l_sphere}]
 dyn_strand_base_r [{shapeType=Sphere, radius=0.5, shapeName=dyn_strand_base_r_sphere}]
 dyn_strand_parent_l [{shapeType=Capsule, length=0.7819, radius=1.4337, shapeName=dyn_strand_parent_l_capsule, center=<1.1275 0 0>, rotation=<rot 90 0 0>}]
 dyn_strand_parent_r [{
   shapeType=Capsule
   length=2.604
   radius=0.8421
   shapeName=dyn_strand_parent_r_capsule
   center=<1.7634 0 -0.2162>
   rotation=<rot 77.6927 0.0002 0.0002>
  }]
 dyn_turtle_l [
  {
   shapeType=Capsule
   length=7.255
   radius=2.255
   shapeName=dyn_turtle_r_capsule
   center=<-0.8315 0.0046 -1.5261>
   rotation=<rot -6.7176 18.8818 -91.1185>
  }
  {shapeType=Capsule, length=6, radius=2, shapeName=dyn_turtle_l_capsule, center=<-1.9218 -7.2031 -1.7988>, rotation=<rot 0 -30 90>}
 ]
 dyn_turtle_r [
  {shapeType=Capsule, length=7.505, radius=1.505, shapeName=dyn_turtle_r_capsule, center=<0 0.4411 -1.2118>, rotation=<rot 70.0001 -90.0002 0>}
  {
   shapeType=Capsule
   length=2
   radius=2.5
   shapeName=dyn_turtle_r_capsule
   center=<-3.477 -5.1367 -1.6755>
   rotation=<rot 20.0003 -59.9999 0.0002>
  }
 ]
 head [{shapeType=Capsule, length=5.6022, radius=6.4559, shapeName=head_capsule, center=<-0.0733 2.5304 0>, rotation=<rot 90 0 0>}]
 neck_01 [{shapeType=Capsule, length=2, radius=3, shapeName=neck_01_capsule, center=<4.6908 7.7143 -3.454>, rotation=<rot 3.5755 4.4239 51.1289>}]
 spine_05 [
  {shapeType=Capsule, length=11.2709, radius=12, shapeName=spine_05_capsule, center=<4.0296 -1.0031 0>}
  {shapeType=Capsule, length=8.75, radius=8, shapeName=spine_05_capsule, center=<6.6183 3.6961 0>, rotation=<rot 90 180 160.0002>}
 ]
 upperarm_l [{shapeType=Capsule, length=11.5, radius=6.005, shapeName=upperarm_l_capsule, center=<8.0891 0 0>, rotation=<rot 90 0 0>}]
 upperarm_r [{shapeType=Capsule, length=11.5, radius=4.75, shapeName=upperarm_l_capsule, center=<-8.0891 0 0>, rotation=<rot 90 0 0>}]
}
constraints [
 {bone1Name=dyn_beanie, bone2Name=dyn_beanie_base, swing1LimitDegrees=5, swing2LimitDegrees=15, twistLimitDegrees=2}
 {bone1Name=dyn_strand_parent_l, bone2Name=dyn_strand_base_l, swing1LimitDegrees=10, swing2LimitDegrees=10, twistLimitDegrees=3}
 {bone1Name=dyn_strand_01_l, bone2Name=dyn_strand_parent_l, swing1LimitDegrees=10, swing2LimitDegrees=10, twistLimitDegrees=3}
 {bone1Name=dyn_strand_02_l, bone2Name=dyn_strand_01_l, swing1LimitDegrees=10, swing2LimitDegrees=10, twistLimitDegrees=3}
 {bone1Name=dyn_hair_parent_l, bone2Name=dyn_hair_base_l, swing1LimitDegrees=5, swing2LimitDegrees=30, twistLimitDegrees=3}
 {bone1Name=dyn_hair_01_l, bone2Name=dyn_hair_parent_l, swing1LimitDegrees=5, swing2LimitDegrees=30, twistLimitDegrees=3}
 {bone1Name=dyn_hair_02_l, bone2Name=dyn_hair_01_l, swing1LimitDegrees=5, swing2LimitDegrees=15, twistLimitDegrees=3}
 {bone1Name=dyn_hair_03_l, bone2Name=dyn_hair_02_l, swing1LimitDegrees=10, swing2LimitDegrees=10, twistLimitDegrees=3}
 {bone1Name=dyn_hair_04_l, bone2Name=dyn_hair_03_l, swing1LimitDegrees=10, swing2LimitDegrees=10, twistLimitDegrees=3}
 {bone1Name=dyn_strand_parent_r, bone2Name=dyn_strand_base_r, swing1LimitDegrees=5, swing2LimitDegrees=5, twistLimitDegrees=3}
 {bone1Name=dyn_strand_01_r, bone2Name=dyn_strand_parent_r, swing1LimitDegrees=5, swing2LimitDegrees=5, twistLimitDegrees=3}
 {bone1Name=dyn_strand_02_r, bone2Name=dyn_strand_01_r, swing1LimitDegrees=5, swing2LimitDegrees=5, twistLimitDegrees=3}
 {bone1Name=dyn_hair_parent_r, bone2Name=dyn_hair_base_r, swing1LimitDegrees=45, swing2LimitDegrees=40, twistLimitDegrees=10}
 {bone1Name=dyn_hair_01_r, bone2Name=dyn_hair_parent_r, swing1LimitDegrees=35, swing2LimitDegrees=30, twistLimitDegrees=5}
 {bone1Name=dyn_hair_02_r, bone2Name=dyn_hair_01_r, swing1LimitDegrees=30, swing2LimitDegrees=20, twistLimitDegrees=5}
 {bone1Name=dyn_hair_03_r, bone2Name=dyn_hair_02_r, swing1LimitDegrees=20, swing2LimitDegrees=20, twistLimitDegrees=5}
 {bone1Name=dyn_hair_04_r, bone2Name=dyn_hair_03_r, swing1LimitDegrees=25, swing2LimitDegrees=35, twistLimitDegrees=5}
 {bone1Name=dyn_hair_05_r, bone2Name=dyn_hair_04_r, swing1LimitDegrees=25, swing2LimitDegrees=25, twistLimitDegrees=5}
 {
  bone1Name=upperarm_l
  bone2Name=clavicle_l
  swing1LimitDegrees=45
  swing1Motion=Free
  swing2LimitDegrees=45
  swing2Motion=Free
  twistLimitDegrees=45
  twistMotion=Free
 }
 {
  bone1Name=upperarm_r
  bone2Name=clavicle_r
  swing1LimitDegrees=45
  swing1Motion=Free
  swing2LimitDegrees=45
  swing2Motion=Free
  twistLimitDegrees=45
  twistMotion=Free
 }
 {
  bone1Name=clavicle_l
  bone2Name=spine_05
  swing1LimitDegrees=45
  swing1Motion=Free
  swing2LimitDegrees=45
  swing2Motion=Free
  twistLimitDegrees=45
  twistMotion=Free
 }
 {
  bone1Name=clavicle_r
  bone2Name=spine_05
  swing1LimitDegrees=45
  swing1Motion=Free
  swing2LimitDegrees=45
  swing2Motion=Free
  twistLimitDegrees=45
  twistMotion=Free
 }
 {
  bone1Name=dyn_collar_l
  bone2Name=spine_05
  swing1LimitDegrees=45
  swing1Motion=Free
  swing2LimitDegrees=45
  swing2Motion=Free
  twistLimitDegrees=45
  twistMotion=Free
 }
 {
  bone1Name=dyn_collar_r
  bone2Name=spine_05
  swing1LimitDegrees=45
  swing1Motion=Free
  swing2LimitDegrees=45
  swing2Motion=Free
  twistLimitDegrees=45
  twistMotion=Free
 }
 {
  bone1Name=dyn_turtle_r
  bone2Name=spine_05
  swing1LimitDegrees=45
  swing1Motion=Free
  swing2LimitDegrees=45
  swing2Motion=Free
  twistLimitDegrees=45
  twistMotion=Free
 }
 {
  bone1Name=dyn_turtle_l
  bone2Name=spine_05
  swing1LimitDegrees=45
  swing1Motion=Free
  swing2LimitDegrees=45
  swing2Motion=Free
  twistLimitDegrees=45
  twistMotion=Free
 }
 {bone1Name=dyn_hair_12_r, bone2Name=dyn_hair_parent_r, swing1LimitDegrees=8, swing2LimitDegrees=8, twistLimitDegrees=2}
 {
  bone1Name=dyn_strand_02_r
  bone2Name=dyn_hair_01_r
  swing1LimitDegrees=45
  swing1Motion=Free
  swing2LimitDegrees=45
  swing2Motion=Free
  twistLimitDegrees=45
  twistMotion=Free
 }
 {
  bone1Name=dyn_strand_02_l
  bone2Name=dyn_hair_02_l
  swing1LimitDegrees=45
  swing1Motion=Free
  swing2LimitDegrees=45
  swing2Motion=Free
  twistLimitDegrees=45
  twistMotion=Free
 }
]
